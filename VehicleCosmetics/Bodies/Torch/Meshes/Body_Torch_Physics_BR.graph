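# Physics asset for the Torch vehicle body cosmetic mesh, the BR variant by its name.
# It defines a single physics body on the root bone, with no collision shapes listed.
# It has no constraints, so it sets up no jointed or ragdoll simulation between bones.
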PhysicsAsset /VehicleCosmetics/Bodies/Torch/Meshes/Body_Torch_Physics_BR
bodies {root []}
constraints []
